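# Unlit translucent pickaxe trail material for Niagara ribbons and mesh particles, with a UseforImpact parameter blending to an alternate impact look.
# Trail shape combines normal-distorted panning cartoon flames with a fire mask, then UV edge fades and noise-threshold erosion.
# Impact shape uses noise-warped fluid flame and an offset gradient timed by a dynamic parameter.
# Both shapes drive curve-atlas color, tinted by particle color, with two smoke layers modulating the trail.
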Material /BRCosmetics/Weapons/FORT_Melee/Pickaxe_Cactus_Rocker_Female/FX/Material/M_Pickaxe_Trail
graph {
 properties {
  bUsedWithNiagaraMeshParticles=true
  bUsedWithNiagaraRibbons=true
  blendMode=BLEND_Translucent
  pixelDepthOffsetMode=PDOM_Legacy
  shadingModel=MSM_Unlit
  usedShadingModels=Unlit
 }
 nodes [
  n1 = ParticleColor()
  n2 = ComponentMask(None=TextureCoordinate()) {r=true}
  n3 = TextureCoordinate()
  n4 = AppendVector(
   A=ScalarParameter() {defaultValue=1, parameterName=UScale}
   B=LinearInterpolate(
    A=ScalarParameter() {defaultValue=0.1, parameterName=Param_1}
    B=ScalarParameter() {defaultValue=1, parameterName=VScale}
    Alpha=ComponentMask(None=@n3) {r=true}
   )
  )
  n5 = MaterialFunctionCall(XY=Multiply(A=TextureCoordinate(), B=@n4)) {materialFunction=&Swizzle}
  n6 = TextureSample(UVs=Panner(Coordinate=TextureCoordinate()) {speedX=1, speedY=0.1}) {
   samplerSource=SSM_Wrap_WorldGroupSettings
   samplerType=SAMPLERTYPE_Normal
   texture=/Game/Effects/Fort_Effects/Gameplay/SmokeDustExplosions/Textures/T_Explosion_NM_01
  }
  n7 = ComponentMask(None=MaterialFunctionCall(Normal=@n6.RGB, Flatness=DynamicParameter() {defaultValue=<rgba 0.9 1 1 1>, paramNames [UVDistortion,Param2,Param3,Param4]}.UVDistortion) {materialFunction=&FlattenNormal}.Result) {g=true, r=true}
  n8 = ScalarParameter() {defaultValue=2.4}
  n9 = TextureSample!(UVs=Panner(Coordinate=Add(A=@n5.YX, B=@n7)) {speedY=2.5}, Tex, Bias=@n8, Apply View MipBias) {
   mipValueMode=TMVM_MipBias
   samplerSource=SSM_Wrap_WorldGroupSettings
   texture=/Game/Effects/Fort_Effects/Textures/Tile/T_Tile_Texture_Cartoon_Flames_Layer
  }
  n10 = MaterialFunctionCall(XY=Multiply(A=@n4, B=TextureCoordinate())) {materialFunction=&Swizzle}
  n11 = TextureSample(UVs=Panner(Coordinate=Add(A=@n7, B=@n10.YX)) {speedY=1.8}, Bias=@n8) {
   mipValueMode=TMVM_MipBias
   samplerSource=SSM_Wrap_WorldGroupSettings
   texture=&T_Tile_Texture_Cartoon_Flames_Layer
  }
  n12 = Multiply(A=LinearInterpolate(Alpha=OneMinus(None=@n2)) {constA=0.5}, B=Add(A=Multiply(A=@n9.R) {constB=0.5}, B=Multiply(A=@n11.R) {constB=0.5}))
  n13 = OneMinus(None=ComponentMask(None=TextureCoordinate()) {r=true})
  n14 = Power(Base=@n13, Exp=ScalarParameter() {defaultValue=3, parameterName=Param_4}) {constExponent=2.5}
  n15 = MaterialFunctionCall(XY=TextureCoordinate()) {materialFunction=&Swizzle}
  n16 = ComponentMask(None=@n15.YX) {g=true}
  n17 = AppendVector(A=ComponentMask(None=@n15.YX) {r=true}, B=Add(A=OneMinus(None=@n16)) {constB=-0.15})
  n18 = ComponentMask(None=@n15.YX) {g=true}
  n19 = AppendVector(A=ComponentMask(None=@n15.YX) {r=true}, B=Add(A=OneMinus(None=@n18)) {constB=0.45})
  n20 = Sine(None=ComponentMask(None=TextureCoordinate()) {g=true}) {period=2}
  n21 = TextureSample(UVs=Panner(Coordinate=LinearInterpolate(A=@n17, B=@n19, Alpha=@n20)), Bias=ScalarParameter() {defaultValue=2, parameterName=Bias}) {mipValueMode=TMVM_MipBias, texture=/Game/Effects/Fort_Effects/Textures/Fire/T_Fire_Mask}
  n22 = LinearInterpolate(
   A=Multiply(A=Add(A=@n12, B=@n14), B=@n21.R)
   B=@n21.R
   Alpha=Power(Base=@n13, Exp=ScalarParameter() {defaultValue=2, parameterName=Param_3}) {constExponent=5}
  )
  n23 = TextureCoordinate()
  n24 = Sine(None=Multiply(A=ComponentMask(None=@n23) {r=true}) {constB=0.5})
  n25 = ComponentMask(None=@n23) {g=true}
  n26 = Multiply(
   A=SphereMask(A=@n23, B=Constant2Vector() {g=0.5, r=1}, Radius=Constant() {r=1}) {hardnessPercent=0}
   B=Sine(None=Multiply(A=@n25) {constB=0.5})
  )
  n27 = Panner(Coordinate=TextureCoordinate() {uTiling=0.7, vTiling=1.5}) {speedX=-1}
  n28 = OneMinus(None=ComponentMask(None=TextureCoordinate()) {r=true})
  n29 = ScalarParameter() {parameterName=Param_2}
  n30 = Constant()
  n31 = Multiply(
   A=Multiply(A=@n22, B=Multiply(A=@n24, B=@n26))
   B=If(
    A=TextureSampleParameter2D(UVs=@n27) {parameterName=Smoke Texture B, texture=&T_NoisePack_01}.R
    B=Power(Base=@n28, Exp=@n29) {constExponent=1.4}
    "A > B"=@n30
    "A == B"=@n30
    "A < B"=Constant() {r=1}
   )
  )
  n32 = Multiply(
   A=TextureCoordinate()
   B=AppendVector(A=ScalarParameter() {defaultValue=1, parameterName=1}, B=ScalarParameter() {defaultValue=0.5, parameterName=2})
  )
  n33 = ComponentMask(None=TextureSample!(UVs=Panner(Coordinate=@n32) {speedX=-0.3}, Tex, Apply View MipBias) {samplerType=SAMPLERTYPE_LinearColor, texture=&T_NoisePack_LinearColor}.RGB) {g=true, r=true}
  n34 = Add(A=Multiply(A=@n33, B=Constant() {r=0.15}), B=TextureCoordinate() {uTiling=2.5, vTiling=0.66})
  n35 = TextureCoordinate() {uTiling=2, vTiling=0.78}
  n36 = Multiply(A=@n33) {constB=0.15}
  n37 = TextureSample!(
   UVs=Panner(
    Coordinate=Add(A=@n35, B=@n36)
    Time=DynamicParameter() {defaultValue=<rgba 0.9 1 1 1>, paramNames [UVDistortion,Param2,Param3,Param4]}.Param2
   ) {speedY=1}
   Tex
   Bias=Constant() {r=2}
   Apply View MipBias
  ) {
   mipValueMode=TMVM_MipBias
   samplerSource=SSM_Clamp_WorldGroupSettings
   texture=/BRCosmetics/Weapons/FORT_Melee/Pickaxe_Cube_Ninja_Male/FX/T_Offset_Gradient_01
  }
  n38 = Multiply(
   A=TextureSample!(UVs=Panner(Coordinate=@n34) {speedY=1}, Tex, Apply View MipBias) {samplerType=SAMPLERTYPE_LinearGrayscale, texture=/BRCosmetics/Weapons/FORT_Melee/Pickaxe_Cube_Ninja_Male/FX/T_FluidFlameTile}.R
   B=Multiply(A=@n37.B)
  )
  n39 = ComponentMask(None=TextureCoordinate()) {g=true}
  n40 = Multiply(A=@n38, B=Sine(None=Multiply(A=@n39) {constB=0.5}))
  DepthFade(Opacity=LinearInterpolate(A=Multiply(A=@n1.A, B=@n31), B=Multiply(A=@n1.A, B=@n40), Alpha=ScalarParameter() {parameterName=UseforImpact})) {fadeDistanceDefault=30}
  n41 = Multiply(
   A=TextureCoordinate()
   B=AppendVector(
    A=ScalarParameter() {defaultValue=0.5, parameterName=Smoke Texture B U}
    B=ScalarParameter() {defaultValue=2, parameterName=Smoke Texture B V}
   )
  )
  n42 = Multiply(
   A=VectorParameter() {parameterName=Smoke Texture B color, defaultValue=<rgba 1 1 1 0>}.RGB
   B=TextureSampleParameter2D(UVs=Panner(Coordinate=@n41) {speedX=1.5}) {parameterName=Smoke Texture B, texture=&T_NoisePack_01}.B
  )
  n43 = ScalarParameter() {parameterName=Smoke Texture B Desaturation}
  n44 = Multiply(
   A=TextureCoordinate()
   B=AppendVector(
    A=ScalarParameter() {defaultValue=0.5, parameterName=Smoke Texture A U}
    B=ScalarParameter() {defaultValue=1, parameterName=Smoke Texture A V}
   )
  )
  n45 = TextureCoordinate()
  n46 = AppendVector(A=ScalarParameter() {defaultValue=1, parameterName=Turbulent U}, B=ScalarParameter() {defaultValue=2.5, parameterName=Turbulent V})
  n47 = TextureSample(UVs=Panner(Coordinate=Multiply(A=@n45, B=@n46)) {speedX=-0.3}) {samplerType=SAMPLERTYPE_LinearColor, texture=&T_NoisePack_LinearColor}
  n48 = Multiply(
   A=AppendVector(A=ComponentMask(None=@n47.RGB) {r=true}, B=ComponentMask(None=@n47.RGB) {g=true})
   B=ScalarParameter() {defaultValue=0.3, parameterName=Turbulent intensity}
  ) {constB=0.2}
  n49 = TextureSampleParameter2D(UVs=Panner(Coordinate=Add(A=@n44, B=@n48)) {speedX=1}) {parameterName=Smoke Texture A, samplerType=SAMPLERTYPE_LinearColor, texture=&T_NoisePack_LinearColor}
  n50 = Multiply(A=Multiply(A=@n49.RGB, B=@n49.RGB), B=VectorParameter() {parameterName=Smoke Texture A color, defaultValue=<rgba 1 1 1 0>}.RGB)
  n51 = ScalarParameter() {parameterName=Smoke Texture A Desaturation}
  n52 = CurveAtlasRowParameter(CurveTime=Clamp(None=@n22)) {
   atlas=&AthenaCurveAtlas01
   curve=/Game/Athena/LinearColorCurves/IceFlamesCurve
   defaultValue=5
   parameterName=Cuve
   sortPriority=0
  }
  n53 = Multiply(A=Add(A=Desaturation(None=@n42, Fraction=@n43), B=Desaturation(None=@n50, Fraction=@n51)), B=@n52)
  n54 = CurveAtlasRowParameter(CurveTime=@n38) {
   atlas=&AthenaCurveAtlas01
   curve=&IceFlamesCurve
   defaultValue=5
   parameterName=Cuve
   sortPriority=0
  }
  MaterialFunctionCall(Emissive=LinearInterpolate(A=Multiply(A=@n53, B=@n1.RGB), B=Multiply(A=@n1.RGB, B=@n54), Alpha=ScalarParameter() {parameterName=UseforImpact})) {materialFunction=&MF_FX_EmissiveAdjustments}
 ]
}
