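# Master material for the Cactus Rocker Female pickaxe, adding an animated, flame-like, audio-reactive emissive glow to the standard character material chain.
# Panned flame noise, distorted by a scrolling normal map, forms two colored layers that combine with a Fresnel tint, masked by texture and local height.
# SampleAudioMPCHistory, from the matching backpack, supplies music amplitude and spectrum.
# UV-derived headstock masks and rotated UVs are exposed through named reroutes.
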Material /BRCosmetics/Weapons/FORT_Melee/Pickaxe_Cactus_Rocker_Female/Materials/M_Pickaxe_Cactus_Rocker_Female_Master
graph {
 properties {
  bAutomaticallySetUsageInEditor=false
  bUseMaterialAttributes=true
  bUsedWithClothing=true
  bUsedWithMeshDeformer=true
  bUsedWithMorphTargets=true
  bUsedWithSkeletalMesh=true
  pixelDepthOffsetMode=PDOM_Legacy
  refractionMethod=RM_IndexOfRefraction
  shadingModel=MSM_FromMaterialExpression
  subsurfaceProfile=&SS_HeroSkin_02
  usedShadingModels=Default Lit | Subsurface | Clear Coat | Subsurface Profile | Cloth | Eye
 }
 nodes [
  n1 = MaterialFunctionCall() {materialFunction=&BaseFN_CharacterFunction_PreFX}
  n2 = GetMaterialAttributes(None=MaterialFunctionCall(
    PreskinnedPosition=@n1.Pre skinned Local Position
    Gmap=@n1.Pre skinned Local Position
    MaterialAttributes=@n1.Pre skinned Local Position
    AO=@n1.Pre skinned Local Position
   ) {materialFunction=&CharacterSkinFX}.Result) {attributeGetTypes [#1, #2]}
  n3 = TextureSampleParameter2D() {
   parameterName=SkinFX_Mask
   samplerSource=SSM_Wrap_WorldGroupSettings
   samplerType=SAMPLERTYPE_Masks
   texture=/BRCosmetics/Weapons/FORT_Melee/Pickaxe_Cactus_Rocker_Female/Textures/T_Pickaxe_Cactus_Rocker_Female_FX
  }
  n4 = Saturate(None=ComponentMask(None=Divide(A=@n1.Pre skinned Local Position) {constB=25}) {g=true})
  n5 = OneMinus(None=@n4)
  n6 = Add(A=Subtract(A=NamedRerouteUsage(), B=Multiply(A=@n5) {constB=0.35}), B=@n4)
  n7 = NamedRerouteUsage()
  n8 = ScalarParameter() {defaultValue=0.05, group=Cactus Rocker, parameterName=UV_Distortion_Intensity_A}
  n9 = Add(A=NamedRerouteUsage(), B=ComponentMask(None=Multiply(A=@n7, B=@n8)) {g=true, r=true})
  n10 = MaterialFunctionCall(X=Constant() {r=1.25}, Y=Constant() {r=3}) {materialFunction=&MakeFloat2}
  n11 = TextureSample(UVs=Panner(Coordinate=Multiply(A=@n9, B=@n10.Result) {constB=1.2}) {speedX=0.06}) {samplerSource=SSM_Wrap_WorldGroupSettings, texture=/Game/Effects/FogVolume/noise_flamy_Tile}
  n12 = NamedRerouteUsage()
  n13 = ScalarParameter() {defaultValue=1.25, group=Cactus Rocker, parameterName=Noise_Max_Brightness}
  n14 = Multiply(
   A=Multiply(
    A=NamedRerouteDeclaration(None=@n11.G) {name=Base_Noise, nodeColor=<rgba 1 0 0.7969 1>}
    B=VectorParameter() {group=Cactus Rocker, parameterName=Color_A, defaultValue=<rgba 0.0466 0.0176 0.5677 0>}.RGB
   )
   B=Add(A=Multiply(A=@n12, B=@n13) {constB=15}) {constB=0.25}
  ) {constA=1}
  n15 = ComponentMask(None=NamedRerouteUsage()) {a=true}
  n16 = Multiply(A=NamedRerouteUsage(), B=ScalarParameter() {defaultValue=0.025, group=Cactus Rocker, parameterName=UV_Distortion_Intensity_B})
  n17 = Multiply(
   A=Add(A=NamedRerouteUsage(), B=ComponentMask(None=@n16) {g=true, r=true})
   B=MaterialFunctionCall(X=Constant() {r=0.75}, Y=Constant() {r=8}) {materialFunction=&MakeFloat2}.Result
  ) {constB=0.5}
  n18 = NamedRerouteUsage()
  n19 = Multiply(
   A=TextureSample(UVs=Panner(Coordinate=@n17) {speedX=0.5}) {samplerSource=SSM_Wrap_WorldGroupSettings, texture=&noise_flamy_Tile}.G
   B=LinearInterpolate(Alpha=ComponentMask(None=@n18) {b=true}) {constA=1.75, constB=2}
  ) {constB=1.75}
  n20 = Saturate(None=Add(A=NamedRerouteUsage(), B=Subtract(A=@n19)))
  n21 = VectorParameter() {group=Cactus Rocker, parameterName=Color_B, defaultValue=<rgba 0 0.3592 0.75 0>}
  n22 = ScalarParameter() {defaultValue=3, group=Cactus Rocker, parameterName=Flame_Brightness}
  n23 = NamedRerouteUsage()
  n24 = Multiply(
   A=Multiply(A=LinearInterpolate(Alpha=@n15) {constA=0.5, constB=2}, B=Multiply(A=@n20, B=@n21.RGB))
   B=Add(A=Multiply(A=@n22, B=@n23)) {constB=1.5}
  )
  n25 = NamedRerouteUsage()
  n26 = ComponentMask(None=@n25) {r=true}
  n27 = ComponentMask(None=@n25) {g=true}
  n28 = Multiply(
   A=VectorParameter() {parameterName=Color_Fresnel, defaultValue=<rgba 0.0303 0.029 0.75 0>}.RGB
   B=Fresnel(ExponentIn=LinearInterpolate(Alpha=@n26) {constA=0.75, constB=2}, BaseReflectFractionIn=Multiply(A=@n27) {constB=0.1}) {baseReflectFraction=0, exponent=2}
  )
  n29 = LinearInterpolate(Alpha=NamedRerouteUsage()) {constA=0.5, constB=10}
  n30 = Multiply(A=Multiply(A=OneMinus(None=@n3.R), B=Saturate(None=@n6)), B=Add(A=Add(A=@n14, B=@n24), B=Multiply(A=@n28, B=@n29)))
  n31 = MaterialFunctionCall(
   AO=@n1.AO
   Distance Blend=@n1.AO
   Fresnel=@n1.AO
   Gmap=@n1.AO
   Pre Skinned Local Position=@n1.AO
   Material=SetMaterialAttributes!(MaterialAttributes=@n2.MaterialAttributes, Base Color, Emissive Color=Add(A=@n2.Emissive Color, B=@n30)) {attributeSetTypes [#1, #2]}
  ) {materialFunction=&BaseFN_CharacterFunction_PostFX}
  MaterialFunctionCall(In=@n31.Result) {materialFunction=&CINE_MF_CharPostFX}
  n32 = Constant() {r=1}
  n33 = Constant() {r=-1}
  n34 = Multiply(A=NamedRerouteUsage(), B=NamedRerouteUsage())
  NamedRerouteDeclaration(None=Rotator(Coordinate=TextureCoordinate(), Time=LinearInterpolate(A=@n32, B=@n33, Alpha=@n34)) {speed=1}) {name=Rotated_UVs, nodeColor=<rgba 0.4531 0 1 1>}
  n35 = Constant() {r=5}
  n36 = Constant() {r=16}
  n37 = Panner(Coordinate=Multiply(
    A=NamedRerouteUsage()
    B=MaterialFunctionCall(X=@n35, Y=@n36) {materialFunction=&MakeFloat2}.Result
   ) {constB=1.2}) {speedX=0.85}
  n38 = TextureSample(UVs=@n37) {
   samplerSource=SSM_Wrap_WorldGroupSettings
   samplerType=SAMPLERTYPE_Normal
   texture=&T_Weap_Distort_N_02
  }
  NamedRerouteDeclaration(None=AppendVector(
    A=ComponentMask(None=@n38.RGB) {g=true, r=true}
    B=TextureSample(UVs=@n37) {samplerSource=SSM_Wrap_WorldGroupSettings, texture=&T_Smooth_tile}.R
   )) {name=Flame-Like_UVDistortion, nodeColor=<rgba 0.0781 0 1 1>}
  n39 = MaterialFunctionCall(Float2=TextureCoordinate()) {materialFunction=&BreakOutFloat2Components}
  n40 = OneMinus(None=Ceil(None=Subtract(A=@n39.R) {constB=0.2}))
  n41 = Subtract(A=@n39.G) {constB=0.8}
  n42 = Multiply(A=@n40, B=OneMinus(None=Ceil(None=@n41)))
  n43 = OneMinus(None=Ceil(None=Subtract(A=@n39.G) {constB=0.25}))
  NamedRerouteDeclaration(None=Saturate(None=Add(A=@n42, B=@n43))) {name=Headstock_Mask, nodeColor=<rgba 1 0.1875 0 1>}
  NamedRerouteDeclaration(None=@n40) {name=Headstock_Split_Mask, nodeColor=<rgba 0.7344 1 0 1>}
  n44 = MaterialFunctionCall!(useSmoothStep=StaticBool(), uvx=Constant() {r=0.5}) {materialFunction=/Game/Accessories/FORT_Backpacks/Backpack_F_MED_Cactus_Rocker/Materials/SampleAudioMPCHistory}
  n45 = NamedRerouteDeclaration(None=@n44.Interpolated Amplitude Music) {name=Audio_Amplitude, nodeColor=<rgba 1 0 0.8438 1>}
  n46 = Power(Base=@n45)
  NamedRerouteDeclaration(None=Saturate(None=Multiply(A=@n46) {constB=2})) {name=Audio_Amplitude_Remap, nodeColor=<rgba 0.6406 0 1 1>}
  NamedRerouteDeclaration(None=@n44.Interpolated Spectrum) {name=Audio_Spectrum, nodeColor=<rgba 0 1 0.2266 1>}
  NamedRerouteDeclaration(None=Multiply(A=@n45) {constB=4}) {name=Audio_Amplitude_Remap_0-0.25, nodeColor=<rgba 0.2188 0 1 1>}
 ]
}
