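# Physics asset for the Cactus Rocker Female pickaxe, a melee weapon cosmetic, defining simplified collision for its axe mesh.
# It holds a single body on the Melee_TwoHanded_Handle bone, shaped as one offset, rotated capsule.
# There are no constraints, so the asset defines no jointed or secondary motion between parts.
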PhysicsAsset /BRCosmetics/Weapons/FORT_Melee/Pickaxe_Cactus_Rocker_Female/Meshes/Cactus_Rocker_Female_Axe_PhysicsAsset
bodies {Melee_TwoHanded_Handle [{shapeType=Capsule, length=71.6914, radius=23.5179, center=<-1.3925 -1.4163 36.3979>, rotation=<rot 1.5889 -90 -0.4647>}]}
constraints []
